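# Material function that builds the surface attributes for the horn attachment of the Octane Lava vehicle skin.
# Base color comes from a texture multiplied by the red channel of a packed RGB map.
# The map's green channel acts as an accent mask, blending base and accent parameters for metallic, roughness, and clear coat thickness and roughness via CustomData0/1.
# A normal map completes the material attributes output.
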MaterialFunction /VehicleCosmetics/Bodies/Octane/Skins/OctaneLava/MF_OctaneLavaAttachment
graph {nodes [
  n1 = TextureSampleParameter2D() {parameterName=AttachmentRGB, texture=&/Skin_OctaneLava_Horn_RGB}
  n2 = TextureSampleParameter2D() {
   parameterName=AttachmentNormal
   samplerType=SAMPLERTYPE_Normal
   texture=&/Skin_OctaneLava_Horn_N
  }
  n3 = SetMaterialAttributes!(
   MaterialAttributes
   BaseColor=Multiply(
    A=TextureSampleParameter2D() {parameterName=AttachmentBaseColor, texture=&/Skin_OctaneLava_Horn_BC}.RGB
    B=@n1.R
   )
   Metallic=LinearInterpolate(
    A=ScalarParameter() {defaultValue=0.8, parameterName=Metallic}
    B=ScalarParameter() {defaultValue=0.4, parameterName=Metallic_Accent}
    Alpha=@n1.G
   )
   Roughness=LinearInterpolate(
    A=ScalarParameter() {defaultValue=0.55, parameterName=Roughness}
    B=ScalarParameter() {defaultValue=0.65, parameterName=Roughness_Accent}
    Alpha=@n1.G
   )
   CustomData0=LinearInterpolate(A=ScalarParameter() {parameterName=CCThickness}, B=ScalarParameter() {defaultValue=0.8, parameterName=CCThickness_Accent}, Alpha=@n1.G) {constA=1, constB=0.2}
   CustomData1=LinearInterpolate(
    A=ScalarParameter() {defaultValue=0.6, parameterName=CCRoughness}
    B=ScalarParameter() {defaultValue=0.8, parameterName=CCRoughness_Accent}
    Alpha=@n1.G
   ) {constA=0.8, constB=0.8}
   Normal=@n2.RGB
  ) {attributeSetTypes [
    #1
    #2
    #3
    #4
    #5
    #6
   ]}
  FunctionOutput(None=@n3)
 ]}
